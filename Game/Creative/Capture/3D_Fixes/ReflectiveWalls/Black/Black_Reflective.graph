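# Black_Reflective is a Default Lit material, named and filed as a black reflective wall surface among Creative capture 3D fixes.
# It holds only four scalar constants, with no textures or parameters, so its inputs are fixed values.
# The pins these constants feed are not recorded, though the name suggests they set a dark, glossy look.
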Material /Game/Creative/Capture/3D_Fixes/ReflectiveWalls/Black/Black_Reflective
graph {
 properties {pixelDepthOffsetMode=PDOM_Legacy, usedShadingModels=Default Lit}
 nodes [
  Constant()
  Constant() {r=1}
  Constant() {r=0.9}
  Constant() {r=0.3}
 ]
}
